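# Niagara muzzle effect for the WinterDecay Mask, built from blue-violet flash, swirl and sparkle bursts.
# Flash layers three sprites and a mesh on one particle, using separate colour and size attributes.
# User.AimVec offsets delayed Sparkles and orients and launches Swirls meshes, rotated by partition and spawn group.
# Camera-distance alphas fade Flash layers and Swirls, Sparkles are distance-culled, and Distortion_LW lists no renderers.
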NiagaraSystem /WinterDecay/Assets/FX/NS_WinterDecay_Mask_Muzzle
dependencies {
 usedDataInterfaces [&NiagaraDataInterfaceCurve, &NiagaraDataInterfaceColorCurve, &NiagaraDataInterfaceCamera]
 usedDynamicInputs []
 usedModules [
  &SystemState
  &self:SystemSpawnScript.SystemScriptSource.SystemScriptGraph.NiagaraNodeAssignment_0.SetVariables_#1
 ]
}
emitters {
 Distortion_LW {rendererClasses [], renderers []}
 Flash {
  emitterSpawnScript [/SetVariables(Emitter.CameraOffset=50, Emitter.SpriteSize:Vector2f=&Multiply_Vector2D_ByFloat(Float=1, Vector2D=<100 90>)) {enabled=false, isSetParameters=true}]
  emitterUpdateScript [
   &EmitterState(Life Cycle Mode=::NewEnumerator1, Loop Behavior=::NewEnumerator1, Loop Delay=0.06, Loop Duration=0.15, UseLoopDelay=true)
   &SpawnBurst_Instantaneous(Spawn Count=1, Use Spawn Probability=false)
   /SetVariables(
    Emitter.Color:LinearColor=<rgba 0.197 0 0.5208 1>
    Emitter.SpriteSize:Vector2f=&Multiply_Vector2D_ByFloat(
     Float=&FloatFromCurve(
      FloatCurve=DataInterface() {propertyValues {Curve {keys [
          {arriveTangent=9.901, interpMode=RCIM_Cubic, leaveTangent=9.901, tangentMode=RCTM_User, time=0, value=1}
          {arriveTangent=12.2238, interpMode=RCIM_Cubic, leaveTangent=12.2238, tangentMode=RCTM_User, time=0.1, value=3.5}
          <at 0.3 0>
         ]}}}
      CurveIndex=$Emitter.Age
     )
     Vector2D=$Emitter.Initial.SpriteSize
    )
   ) {enabled=false, isSetParameters=true}
  ]
  particleSpawnScript [
   /Niagara/Modules/Spawn/Initialization/InitializeParticle(
    Color=<rgba 1.9406 1.8679 2 0.3>
    Lifetime=0.3
    Mesh Scale=&VectorFromFloat(Value=&Lerp_Float(A=1, B=0.9, Alpha=&MakeFloatFromInt(Int=&SpawnGroup())))
    Module.Write Color=true
    Module.Write Lifetime=true
    Module.Write Mass=false
    Module.Write Material Random=true
    Module.Write Position=false
    Module.Write Scale=true
    Module.Write SpriteRotation=false
    Module.Write SpriteSize=false
    Position=&SimulationPosition()
    Sprite Size=<5 5>
   ) {enabled=false}
   /Niagara/Modules/Spawn/Initialization/V2/InitializeParticle(
    Color=<rgba 0.13 0.325 1.3 1>
    Lifetime=0.3
    Mesh Scale Mode=::NewEnumerator4
    Position Offset=<40 0 0>
    Sprite Size Mode=::NewEnumerator3
    Uniform Sprite Size=60
    UsePositionOffset=false
    Write Material Random=true
   ) {module=InitializeParticle001}
   &CameraOffset(Camera Offset Amount=-5)
   /SetVariables(Particles.Position=&AddVectorToPosition(Position=<0 0 0>, Vector=&Multiply_VectorByFloat(Vector=$User.AimVec, Float=28))) {enabled=false, isSetParameters=true}
   &InitialMeshOrientation(Mesh Axis To Orient=<0 0 1>, Mesh Orientation Mode=::NewEnumerator2, Orientation Vector=$User.AimVec, Use Rotation Vector=false) {enabled=false}
   &InitialMeshOrientation(Mesh Axis To Orient=<0 0 1>, Mesh Orientation Mode=::NewEnumerator2, Use Rotation Vector=false) {module=InitialMeshOrientation001}
   /SetVariables(Particles.ColorGlow:LinearColor=<rgba 0.3418 0.9104 1 1>) {isSetParameters=true}
  ]
  particleUpdateScript [
   &SolveForcesAndVelocity(Clamp Velocity=false, Limit Acceleration=false)
   &ParticleState()
   &ScaleColor(
    Scale Alpha=&FloatFromCurve(
     FloatCurve=DataInterface() {propertyValues {Curve {keys [
         {arriveTangent=-1.8753, interpMode=RCIM_Cubic, leaveTangent=-1.8753, tangentMode=RCTM_User, time=0.25, value=1}
         {interpMode=RCIM_Cubic, tangentMode=RCTM_User, time=1, value=0}
        ]}}}
     Scale Curve=&Multiply_Float(A=30, B=$System.CamDistanceAlpha)
    )
    Scale RGB=&VectorFromFloat(Value=10)
    ScaleA=true
    ScaleRGB=true
   )
   /SetVariables(Particles.Color2=&ColorFromCurve(ColorCurve=DataInterface() {propertyValues {
       AlphaCurve {keys [<at 0 1 RCIM_Cubic>,<at 1 0 RCIM_Cubic>]}
       BlueCurve {keys [<at 0 1 RCIM_Cubic>]}
       GreenCurve {keys [<at 0 0.7725 RCIM_Cubic>]}
       RedCurve {keys [<at 0 0.9467 RCIM_Cubic>]}
      }})) {isSetParameters=true}
   /SetVariables(Particles.SpriteSizeFlare:Vector2f=&Multiply_Vector2D_ByFloat(
     Float=&FloatFromCurve(FloatCurve=DataInterface() {propertyValues {Curve {keys [
          {interpMode=RCIM_Cubic, tangentMode=RCTM_User, time=0, value=0.2}
          {arriveTangent=-1.6444, interpMode=RCIM_Cubic, leaveTangent=-1.6444, tangentMode=RCTM_User, time=0.3, value=1}
          {arriveTangent=-1.1616, interpMode=RCIM_Cubic, leaveTangent=-1.1616, tangentMode=RCTM_User, time=1, value=0}
         ]}}})
     Vector2D=<100 100>
    )) {isSetParameters=true}
   /SetVariables(
    Particles.CameraOffsetFlare=10
    Particles.ColorFlare:LinearColor=&MultiplyLinearColor(
     Linear Color=<rgba 1.4 1.6 2 1>
     Scale Factor=&MakeVector4FromVectorAndFloat(
      "Vector (RGB)"=<1 1 1>
      "Float (Alpha)"=&FloatFromCurve(
       FloatCurve=DataInterface() {propertyValues {Curve {keys [<at 0 0>,<at 0.2041 1>,<at 0.71 1>,<at 1 0>]}}}
       Scale Curve=$System.CamDistanceAlpha
      )
     )
    )
   ) {isSetParameters=true}
   /SetVariables(
    Particles.ColorDark:LinearColor=&MultiplyLinearColor(
     Linear Color=&ColorFromCurve(ColorCurve=DataInterface() {propertyValues {
        AlphaCurve {keys [<at 0 0>, {arriveTangent=3.5887, interpMode=RCIM_Cubic, leaveTangent=3.5887, time=0.1665, value=4}, <at 0.5573 2>, <at 1 0 RCIM_Cubic>]}
        BlueCurve {keys [<at 0 1 RCIM_Cubic>]}
        GreenCurve {keys [<at 0 0.2 RCIM_Cubic>]}
        RedCurve {keys [<at 0 0.04 RCIM_Cubic>]}
       }})
     Scale Factor=&MakeVector4FromVectorAndFloat("Vector (RGB)"=<1 1 1>, "Float (Alpha)"=$System.CamDistanceAlpha)
    )
    Particles.SpriteSizeDark:Vector2f=&Vector2DFromFloat(Value=&Lerp_Float(
      A=0
      B=200
      Alpha=&FloatFromCurve(FloatCurve=DataInterface() {propertyValues {Curve {keys [
           {interpMode=RCIM_Cubic, tangentMode=RCTM_User, time=0, value=0}
           {interpMode=RCIM_Cubic, tangentMode=RCTM_User, time=0.1, value=1}
           {arriveTangent=-1.1699, interpMode=RCIM_Cubic, leaveTangent=-1.1699, tangentMode=RCTM_User, time=0.7, value=0.4}
          ]}}})
     ))
   ) {isSetParameters=true}
   &DynamicMaterialParameters(
    Index 0 Param 1=&FloatFromCurve(FloatCurve=DataInterface() {propertyValues {Curve {keys [<at 0 0>,<at 1 1>]}}}, Scale Curve=3)
    Index 0 Param 3=&FloatFromCurve(FloatCurve=DataInterface() {propertyValues {Curve {keys [<at 0 1>,<at 0.5 0>]}}}, Scale Curve=3)
    Index 0 Param 4=1
    Index 3 Param 1=1
    Module.Index0Float0Write=false
    Module.Index0Float1Write=false
    Module.Index0Float2Write=true
    Module.Index0Float3Write=false
    Module.Index3Float0Write=true
    Module.Index3Float1Write=true
    Module.Index3Float2Write=true
    Module.Index3Float3Write=true
    Write Parameter Index 3=::NewEnumerator1
   )
   &ScaleMeshSize(Scale Factor=&VectorFromFloat(Value=&FloatFromCurve(FloatCurve=DataInterface() {propertyValues {Curve {keys [
          {arriveTangent=1.755, interpMode=RCIM_Cubic, leaveTangent=1.755, tangentMode=RCTM_User, time=0, value=0.2}
          {interpMode=RCIM_Cubic, tangentMode=RCTM_User, time=1, value=1}
         ]}}})))
   /SetVariables(Particles.ColorGlow:LinearColor=&MultiplyLinearColor(
     Linear Color=&ColorFromCurve(ColorCurve=DataInterface() {propertyValues {
        AlphaCurve {keys [<at 0 0>,<at 0.0766 1 RCIM_Cubic>,<at 0.8142 0 RCIM_Cubic>]}
        BlueCurve {keys [<at 0.1402 6 RCIM_Cubic>,<at 0.2816 1.5>,<at 0.7623 1 RCIM_Cubic>]}
        GreenCurve {keys [<at 0.1402 2.5599 RCIM_Cubic>,<at 0.2816 0.3125>,<at 0.7623 0.15 RCIM_Cubic>]}
        RedCurve {keys [<at 0.1402 0.1027 RCIM_Cubic>,<at 0.2816 0.075>,<at 0.7623 0.1783 RCIM_Cubic>]}
       }})
     Scale Factor=&MakeVector4FromVectorAndFloat("Vector (RGB)"=<1 1 1>, "Float (Alpha)"=$System.CamDistanceAlpha)
    )) {isSetParameters=true}
  ]
  rendererClasses [&NiagaraSpriteRendererProperties, &NiagaraMeshRendererProperties]
  renderers [
   {rendererClass=&NiagaraSpriteRendererProperties}
   {rendererClass=&NiagaraSpriteRendererProperties, rendererIndex=1}
   {rendererClass=&NiagaraSpriteRendererProperties, rendererIndex=2}
   {rendererClass=&NiagaraMeshRendererProperties, rendererIndex=3}
  ]
 }
 Sparkles {
  emitterUpdateScript [
   &EmitterState(
    Enable Distance Culling=true
    Life Cycle Mode=::NewEnumerator1
    Loop Behavior=::NewEnumerator1
    Loop Duration=1
    MaxDistance=2500
    Module.bUseMaxDistance=true
    Module.bUseMinDistance=false
    Scalability Mode=::NewEnumerator1
    UseLoopDelay=false
   )
   &SpawnBurst_Instantaneous(
    Spawn Count=&RandomRangeInt(Minimum=20, Maximum=40)
    Spawn Time=0.15
    Use Spawn Probability=false
    UseLoopCountLimit=false
   )
  ]
  particleSpawnScript [
   &InitializeParticle(
    Color=<rgba 74.7792 74.7792 1000 1>
    Lifetime=&RandomRangeFloat(Minimum=0.3, Maximum=0.7)
    Sprite Size Mode=::NewEnumerator3
    Uniform Sprite Size=&Lerp_Float(
     A=5
     B=40
     Alpha=&FloatFromCurve(
      FloatCurve=DataInterface() {propertyValues {Curve {keys [
          {interpMode=RCIM_Cubic, tangentMode=RCTM_User, time=0, value=0}
          {arriveTangent=1.6093, interpMode=RCIM_Cubic, leaveTangent=1.6093, tangentMode=RCTM_User, time=1, value=1}
         ]}}}
      CurveIndex=&RandomRangeFloat(Minimum=0, Maximum=1)
     )
    )
    UsePositionOffset=false
    Write Material Random=true
   )
   &ShapeLocation(
    MaskSpawnGroup=false
    Offset=&Multiply_VectorByFloat(Vector=$User.AimVec, Float=300)
    Offset Mode=::NewEnumerator0
    Override Seed=false
    UseOwnerScale=true
   )
   &AddVelocity(
    Module.Use Falloff Distance=false
    Override Seed=false
    Velocity Mode=::NewEnumerator1
    Velocity Speed=&RandomRangeFloat(Minimum=800, Maximum=1500)
   )
  ]
  particleUpdateScript [
   &ParticleState()
   &GravityForce(Gravity=&Multiply_VectorByFloat(Vector=<0 0 -980>, Float=&RandomRangeFloat(Minimum=0.1, Maximum=0.2)))
   &InheritVelocity(
    Inherited Velocity Amount Scale=&VectorFromFloat(Value=0.1)
    Inherited Velocity Speed Limit:NiagaraFloat=&RandomRangeFloat(Minimum=10, Maximum=30)
    Speed Limit=true
    UseSpeedThreshold=false
   )
   &ScaleColor(
    Scale Alpha=&FloatFromCurve(FloatCurve=DataInterface() {propertyValues {Curve {keys [
         {interpMode=RCIM_Cubic, tangentMode=RCTM_User, time=0, value=0}
         {interpMode=RCIM_Cubic, tangentMode=RCTM_User, time=0.2, value=1}
         {interpMode=RCIM_Cubic, tangentMode=RCTM_User, time=1, value=0}
        ]}}})
    ScaleA=true
    ScaleRGB=true
   )
   /SetVariables(Particles.ColorGlow:LinearColor=&MultiplyLinearColor(
     Linear Color=$Particles.Color
     Scale Factor=&MakeVector4_FromLinearColor(LinearColor=<rgba 0.1172 0.1014 0.3 0.01>)
    )) {isSetParameters=true}
   &Drag(Drag=&RandomRangeFloat(Minimum=1, Maximum=2), Use Linear Drag=true, Use Rotational Drag=true)
   &SolveForcesAndVelocity(Clamp Velocity=false, Limit Acceleration=false)
  ]
  rendererClasses [&NiagaraSpriteRendererProperties]
  renderers [
   {rendererClass=&NiagaraSpriteRendererProperties}
   {rendererClass=&NiagaraSpriteRendererProperties, rendererIndex=1}
  ]
 }
 Swirls {
  emitterUpdateScript [
   &EmitterState(Life Cycle Mode=::NewEnumerator1, Loop Behavior=::NewEnumerator1, Loop Delay=0.1, Loop Duration=0.3, UseLoopDelay=true)
   &SpawnRate(SpawnRate=100, Use Spawn Probability=false) {enabled=false}
   &SpawnBurst_Instantaneous(Spawn Count=3, Use Spawn Probability=false, UseLoopCountLimit=false)
   &SpawnBurst_Instantaneous(Spawn Count=3, Spawn Group=1, Use Spawn Probability=false, UseLoopCountLimit=false) {module=SpawnBurst_Instantaneous001}
   /SetVariables(Emitter.FadeInLifetime:NiagaraFloat=&FloatFromCurve(FloatCurve=DataInterface() {propertyValues {Curve {keys [<at 0 0>,<at 0.2 1>]}}}, CurveIndex=$Engine.System.Age)) {isSetParameters=true}
  ]
  particleSpawnScript [
   /SetVariables(Particles.VisibilityTag=&SpawnGroup()) {isSetParameters=true}
   &PartitionParticles(Number Of Partitions=3)
   &InitializeParticle(
    Color=<rgba 0.85 0.9 1 0.33>
    Lifetime=0.7
    Mesh Scale=&Multiply_VectorByFloat(Vector=<1 1 0.4>, Float=4)
    Mesh Scale Mode=::NewEnumerator4
    Position Offset=&Multiply_VectorByFloat(Vector=$User.AimVec, Float=300)
    Sprite Size Mode=::NewEnumerator3
    Uniform Sprite Size=2
    UsePositionOffset=true
    Write Material Random=true
   )
   &InitialMeshOrientation(
    Mesh Axis To Orient=<0 0 -1>
    Mesh Orientation Mode=::NewEnumerator2
    Orientation Vector=$User.AimVec
    Rotation=&MakeVector(
     X=0
     Y=0
     Z=&Add_Float(
      A=&MergeHelper_FloatSwitch(0=0, 1=0.33, 2=0.66, 3=?, Index=$Particles.PartitionParticles.Partition, Num=::NewEnumerator1)
      B=&MergeHelper_FloatSwitch(0=0, 1=0.02, 2=?, 3=?, Index=&SpawnGroup(), Num=::NewEnumerator0)
     )
    )
    Use Rotation Vector=true
   )
   &AddVelocity(Velocity=$User.AimVec, Velocity Speed Scale=700)
  ]
  particleUpdateScript [
   /SetVariables(Particles.CamDistanceAlphaSplines:NiagaraFloat=&RemapRange(
     Input Value To Remap=/Niagara/DynamicInputs/Camera/DistanceToCamera(Use Owner Position:NiagaraBool=true, Camera Query:NiagaraDataInterfaceCamera=DataInterface())
     Input Min=400
     Input Max=600
     Output Min=0
     Output Max=1
    )) {isSetParameters=true}
   &ParticleState()
   /SetVariables(Particles.XY:NiagaraFloat=&FloatFromCurve(FloatCurve=DataInterface() {propertyValues {Curve {
        defaultValue=1
        keys [
         {arriveTangent=0.3602, interpMode=RCIM_Cubic, leaveTangent=0.3602, tangentMode=RCTM_User, time=0, value=0.5}
         {arriveTangent=0.5868, interpMode=RCIM_Cubic, leaveTangent=0.5868, tangentMode=RCTM_User, time=0.8, value=1}
        ]
       }}})) {isSetParameters=true}
   &ScaleMeshSize(Scale Factor=&MakeVector(
     X=$Particles.XY
     Y=$Particles.XY
     Z=&FloatFromCurve(FloatCurve=DataInterface() {propertyValues {Curve {keys [<at 0 0.5>,<at 1 1>]}}})
    ))
   &DynamicMaterialParameters(
    Index 0 Param 1=&FloatFromCurve(FloatCurve=DataInterface() {propertyValues {Curve {keys [<at 0 -0.5>,<at 1 0.5>]}}})
    Module.Index0Float0Write=true
    Module.Index0Float1Write=false
    Module.Index0Float2Write=false
    Module.Index0Float3Write=false
   )
   &ScaleColor(
    Scale Alpha=&FloatFromCurve(
     FloatCurve=DataInterface() {propertyValues {Curve {keys [
         {arriveTangent=9.6441, interpMode=RCIM_Cubic, leaveTangent=9.6441, tangentMode=RCTM_User, time=0, value=0}
         {arriveTangent=-1.2848, interpMode=RCIM_Cubic, leaveTangent=-1.2848, tangentMode=RCTM_User, time=0.2, value=1}
         {arriveTangent=-0.7928, interpMode=RCIM_Cubic, leaveTangent=-0.7928, tangentMode=RCTM_User, time=1, value=0}
        ]}}}
     Scale Curve=$Particles.CamDistanceAlphaSplines
    )
    Scale RGB=&MakeVectorFromLinearColorRGB(Color=&ColorFromCurve(ColorCurve=DataInterface() {propertyValues {
        AlphaCurve {keys [<at 0 1 RCIM_Cubic>]}
        BlueCurve {keys [<at 0.0233 1 RCIM_Cubic>,<at 0.5032 0.56 RCIM_Cubic>]}
        GreenCurve {keys [<at 0.0233 0.736 RCIM_Cubic>,<at 0.5032 0.252 RCIM_Cubic>]}
        RedCurve {keys [<at 0.0233 0.3395 RCIM_Cubic>,<at 0.5032 0.3547 RCIM_Cubic>]}
       }}))
    ScaleA=true
    ScaleRGB=true
   )
   &Drag(Drag=1.1, Use Linear Drag=true, Use Rotational Drag=true)
   &SolveForcesAndVelocity(Clamp Velocity=false, Limit Acceleration=false)
   /SetVariables(Particles.ColorGlow:LinearColor=&MultiplyLinearColor(
     Linear Color=$Particles.Color
     Scale Factor=&MakeVector4_FromLinearColor(LinearColor=<rgba 2.1 2.45 4.2 10>)
    )) {isSetParameters=true}
   /SetVariables(Particles.ColorSharp:LinearColor=&MultiplyLinearColor(
     Linear Color=$Particles.Color
     Scale Factor=&MakeVector4FromVectorAndFloat(
      "Vector (RGB)"=&MakeVectorFromLinearColorRGB(Color=<rgba 0.05 0.05 5 1>)
      "Float (Alpha)"=&FloatFromCurve(
       FloatCurve=DataInterface() {propertyValues {Curve {keys [{interpMode=RCIM_Cubic, tangentMode=RCTM_User, time=0, value=1}, <at 0.5 0>]}}}
       Scale Curve=5
      )
     )
    )) {isSetParameters=true}
  ]
  rendererClasses [&NiagaraMeshRendererProperties]
  renderers [
   {rendererClass=&NiagaraMeshRendererProperties, rendererIndex=0}
   {rendererClass=&NiagaraMeshRendererProperties, rendererIndex=1}
   {rendererClass=&NiagaraMeshRendererProperties, rendererIndex=2}
  ]
 }
}
summary {
 systemName=NS_WinterDecay_Mask_Muzzle
 emitters [
  {emitterName=Flash, rendererClasses [&NiagaraSpriteRendererProperties, &NiagaraMeshRendererProperties]}
  {emitterName=Distortion_LW, rendererClasses []}
  {emitterName=Swirls, rendererClasses [&NiagaraMeshRendererProperties]}
  {emitterName=Sparkles, rendererClasses [&NiagaraSpriteRendererProperties]}
 ]
 userVariables [
  User.AimVec {type=Vector3f, description="", defaultValue {struct=Vector3f, value=<0 1 1>}}
  User.Global Scale {type=NiagaraFloat, defaultValue=1, description=""}
  User.Variant {type=NiagaraInt32, defaultValue=1}
 ]
}
