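# Niagara system for the WinterDecay mask that shows mesh tendrils pointing toward a teleport portal.
# DirectionTendrils continuously spawns short-lived meshes, partitioned across six socket indices, re-snaps them to those sockets each update and faces them just above User.PortalPosition.
# Curves animate scale, color fade and dynamic material parameters, with separate dark color and parameter attributes alongside two mesh renderers.
# Distance-based stretching is disabled, so AttachDistance has no visible effect here.
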NiagaraSystem /WinterDecay/Assets/FX/NS_WinterDecay_Mask_TeleportDirection
dependencies {
 usedDataInterfaces [
  &NiagaraDataInterfaceCurve
  &NiagaraDataInterfaceSkeletalMesh
  &NiagaraDataInterfaceColorCurve
  &NiagaraDataInterfaceSocketReader
 ]
 usedDynamicInputs [
  /Game/Effects/Niagara/DynamicInputs/State/FN_IsExecutionStateActive
  &NormalizeVector
  &Subtract_Position
  /Niagara/DynamicInputs/Helpers/SimulationPosition
  &RemapRange
 ]
 usedModules [
  &self:SystemSpawnScript.SystemScriptSource.SystemScriptGraph.NiagaraNodeAssignment_4.SetVariables_#1
  &SystemState
  &Timeline
  &self:SystemSpawnScript.SystemScriptSource.SystemScriptGraph.NiagaraNodeAssignment_0.SetVariables_#2
  &self:SystemSpawnScript.SystemScriptSource.SystemScriptGraph.NiagaraNodeAssignment_1.SetVariables_#3
  &self:SystemSpawnScript.SystemScriptSource.SystemScriptGraph.NiagaraNodeAssignment_2.SetVariables_#4
 ]
}
emitters {DirectionTendrils {
  emitterUpdateScript [
   &EmitterState(
    Enable Distance Culling=true
    Inactive Response=::NewEnumerator1
    Life Cycle Mode=::NewEnumerator1
    Loop Behavior=::NewEnumerator1
    Loop Duration Mode=::NewEnumerator1
    MaxDistance=3200
    Module.bUseMaxDistance=true
    Module.bUseMinDistance=false
    Scalability Mode=::NewEnumerator1
    UseLoopDelay=false
   )
   &SpawnRate(SpawnRate=2, Use Spawn Probability=false)
  ]
  particleSpawnScript [
   &PartitionParticles(Number Of Partitions=6)
   &InitializeParticle(
    Color=<rgba 0 0.6667 1 1>
    Lifetime=2
    Mesh Scale Mode=::NewEnumerator3
    Mesh Uniform Scale=&RandomRangeFloat(Minimum=0.8, Maximum=1.1)
    UsePositionOffset=false
    Write Material Random=true
   )
   &SocketLocation(
    Orientation Sampling=::NewEnumerator1
    SamplePosOffset=false
    Socket Index=$Particles.PartitionParticles.Partition
    Socket Reader=$System.Socket Reader Interface
    Socket Sampling Mode=::NewEnumerator3
    VelocitySpeedLimit=false
   ) {enabled=false, module=SocketLocation001}
  ]
  particleUpdateScript [
   &ParticleState()
   &ScaleMeshSize(Scale Factor=&Multiply_VectorByFloat(
     Vector=&Multiply_VectorByFloat(
      Vector=&MakeVector(
       X=&Lerp_Float(
        A=0.3
        B=0.6
        Alpha=&FloatFromCurve(FloatCurve=DataInterface() {propertyValues {Curve {keys [
             {interpMode=RCIM_Cubic, tangentMode=RCTM_User, time=0, value=0}
             {arriveTangent=2.422, interpMode=RCIM_Cubic, leaveTangent=2.422, tangentMode=RCTM_User, time=0.7, value=1}
            ]}}})
       )
       Y=&FloatFromCurve(
        FloatCurve=DataInterface() {propertyValues {Curve {keys [
            {arriveTangent=-1.8798, interpMode=RCIM_Cubic, leaveTangent=-1.8798, tangentMode=RCTM_User, time=0, value=1}
            {interpMode=RCIM_Cubic, tangentMode=RCTM_User, time=0.8, value=0.3}
           ]}}}
        Scale Curve=0.85
       )
       Z=&FloatFromCurve(
        FloatCurve=DataInterface() {propertyValues {Curve {keys [
            {arriveTangent=-1.862, interpMode=RCIM_Cubic, leaveTangent=-1.862, tangentMode=RCTM_User, time=0, value=1}
            {interpMode=RCIM_Cubic, tangentMode=RCTM_User, time=0.8, value=0.3}
           ]}}}
        Scale Curve=0.85
       )
      )
      Float=-0.2
     )
     Float=1
    )) {module=ScaleMeshSize001}
   &SocketLocation(
    Orientation Sampling=::NewEnumerator1
    SamplePosOffset=false
    Sampled Orientation Coordinate Space:ENiagaraCoordinateSpace=::World
    Socket Index=$Particles.PartitionParticles.Partition
    Socket Reader=$System.Socket Reader Interface
    Socket Sampling Mode=::NewEnumerator3
    VelocitySpeedLimit=false
   )
   &SkeletalMeshLocation(
    Bone Sampling Mode=::NewEnumerator4
    MaskSpawnGroup=false
    Orientation Sampling=::NewEnumerator1
    Sampled Position Offset=<5 0 0>
    Skeletal Mesh=DataInterface() {propertyValues {
      DefaultMesh=null
      ExcludeBoneName=root
      OptionalComponentClass=null
      PreviewMesh=&M_Med_Soldier_04_Preview
      SkinningMode=SkinOnTheFly
      SoftSourceActor=null
      SourceMode=Default
      UvSetIndex=0
      WholeMeshLOD=-1
      bExcludeBone=false
      bReadDeformedGeometry=true
      bRequireCurrentFrameData=true
      ComponentTags []
      FilteredBones [head]
      FilteredSockets [AttachSocket]
      MeshUserParameter {parameter {name=None, typeDefHandle {registeredTypeIndex=1}, varData []}}
      SamplingRegions []
     }}
   ) {enabled=false}
   /SetVariables(Particles.AttachDistance:NiagaraFloat=&DistanceBetweenPositions(A=$User.PortalPosition, B=$Output.SocketLocation.SampledPosition)) {isSetParameters=true}
   &ScaleColor(
    Scale Alpha=$System.Timeline.NormalizedTime
    Scale RGB=&MakeVectorFromLinearColorRGB(Color=&ColorFromCurve(ColorCurve=DataInterface() {propertyValues {
        AlphaCurve {keys [<at 0 1 RCIM_Cubic>]}
        BlueCurve {keys [<at 0 1 RCIM_Cubic>, {arriveTangent=-0.98, interpMode=RCIM_Cubic, leaveTangent=-0.98, time=0.4758, value=1}, <at 1 0.02>]}
        GreenCurve {keys [<at 0 1 RCIM_Cubic>, {arriveTangent=-0.986, interpMode=RCIM_Cubic, leaveTangent=-0.986, time=0.4758, value=0.6}, <at 1 0.014>]}
        RedCurve {keys [<at 0 1 RCIM_Cubic>, {arriveTangent=-0.9852, interpMode=RCIM_Cubic, leaveTangent=-0.9852, time=0.4758, value=0.6538}, <at 1 0.0148>]}
       }}))
    ScaleA=false
    ScaleRGB=true
   ) {enabled=false}
   &ScaleMeshSize(Scale Factor=&Multiply_Vector(
     A=&Lerp_Vector(
      A=<0.1 1.5 1.5>
      B=<0.3 0.5 0.5>
      Alpha=&FloatFromCurve(
       FloatCurve=DataInterface() {propertyValues {Curve {keys [
           {arriveTangent=2.7407, interpMode=RCIM_Cubic, leaveTangent=2.7407, tangentMode=RCTM_User, time=0, value=0}
           {interpMode=RCIM_Cubic, tangentMode=RCTM_User, time=1, value=1}
          ]}}}
       CurveIndex=$Emitter.Age
      )
     )
     B=&MakeVector(
      X=&Max_Float(
       A:NiagaraFloat=2
       B:NiagaraFloat=&MakeCustomFloatFromBool(
        Bool=&SetBoolByFloatComparison(A=1000, B=$Particles.AttachDistance)
        True Float=&Divide_Float(A=$Particles.AttachDistance, B=100)
        False Float=10
       )
      )
      Y=1
      Z=1
     )
    )) {enabled=false}
   &ScaleColor(
    Scale Alpha=&FloatFromCurve(FloatCurve=DataInterface() {propertyValues {Curve {keys [
         {interpMode=RCIM_Cubic, tangentMode=RCTM_User, time=0, value=0}
         {interpMode=RCIM_Cubic, tangentMode=RCTM_User, time=0.2, value=1}
         <at 1 0 RCIM_Cubic>
        ]}}})
    Scale RGB=&MakeVectorFromLinearColorRGB(Color=&ColorFromCurve(ColorCurve=DataInterface() {propertyValues {
        AlphaCurve {keys [<at 0 1 RCIM_Cubic>]}
        BlueCurve {keys [{arriveTangent=-3.8788, interpMode=RCIM_Cubic, time=0.0609, value=1.3}, <at 0.821 0.5 RCIM_Cubic>]}
        GreenCurve {keys [{arriveTangent=-0.6206, interpMode=RCIM_Cubic, time=0.0609, value=1.3}, <at 0.821 0.5 RCIM_Cubic>]}
        RedCurve {keys [{arriveTangent=-3.4727, interpMode=RCIM_Cubic, time=0.0609, value=0.39}, <at 0.821 0.5 RCIM_Cubic>]}
       }}))
    ScaleA=true
    ScaleRGB=true
   ) {module=ScaleColor001}
   &UpdateMeshOrientation(
    Facing Interpolation=false
    Facing Position:NiagaraPosition=&AddVectorToPosition(Position=$User.PortalPosition, Vector=<0 0 20>)
    Orientation Method=::NewEnumerator2
    Override Quat=false
    TransformFacingOrigin=false
    TransformFacingPosition=false
    TransformUp=false
   ) {module=UpdateMeshOrientation001}
   &CameraOffset(Camera Offset Amount=-3)
   /SetVariables(Particles.Position=&AddVectorToPosition(
     Position=$Particles.Position
     Vector=&Multiply_VectorByFloat(
      Vector=&RotateVectorByQuaternion(Vector To Rotate=<1 0 0>, Quaternion=$Particles.MeshOrientation)
      Float=&Lerp_Float(
       A=0
       B=0
       Alpha=&FloatFromCurve(FloatCurve=DataInterface() {propertyValues {Curve {keys [<at 0 0>,<at 1 1>]}}})
      )
     )
    )) {isSetParameters=true}
   /SetVariables(Particles.ColorDark:LinearColor=&MakeLinearColorFromVectorAndFloat(
     "Vector (RGB)"=&MakeVectorFromLinearColorRGB(Color=&ColorFromCurve(ColorCurve=DataInterface() {propertyValues {
         AlphaCurve {keys [<at 0 1 RCIM_Cubic>]}
         BlueCurve {keys [<at 0 0.3 RCIM_Cubic>,<at 0.6877 0.6>]}
         GreenCurve {keys [<at 0 0.2 RCIM_Cubic>,<at 0.6877 0.12>]}
         RedCurve {keys [<at 0 0 RCIM_Cubic>,<at 0.6877 0.2>]}
        }}))
     "Float (Alpha)"=&FloatFromCurve(
      FloatCurve=DataInterface() {propertyValues {Curve {keys [
          {arriveTangent=26.9411, interpMode=RCIM_Cubic, leaveTangent=26.9411, tangentMode=RCTM_User, time=0, value=0}
          {interpMode=RCIM_Cubic, tangentMode=RCTM_User, time=0.1, value=1}
          {arriveTangent=-1.2473, interpMode=RCIM_Cubic, leaveTangent=-1.2473, tangentMode=RCTM_User, time=1, value=0}
         ]}}}
      Scale Curve=2
     )
    )) {isSetParameters=true}
   &DynamicMaterialParameters(
    Index 0 Param 1=-0.6
    Index 0 Param 2=&Lerp_Float(
     A=-0.2
     B=0.6
     Alpha=&FloatFromCurve(FloatCurve=DataInterface() {propertyValues {Curve {keys [
          {arriveTangent=2.3857, interpMode=RCIM_Cubic, leaveTangent=2.3857, tangentMode=RCTM_User, time=0, value=0}
          {arriveTangent=0.2209, interpMode=RCIM_Cubic, leaveTangent=0.2209, tangentMode=RCTM_User, time=1, value=1}
         ]}}})
    )
    Index 0 Param 3=&Lerp_Float(
     A=1
     B=2
     Alpha=&FloatFromCurve(FloatCurve=DataInterface() {propertyValues {Curve {keys [<at 0 0>,<at 0.7 1>]}}})
    )
    Index 0 Param 4=&FloatFromCurve(FloatCurve=DataInterface() {propertyValues {Curve {keys [
         {interpMode=RCIM_Cubic, tangentMode=RCTM_User, time=0.1, value=1}
         {interpMode=RCIM_Cubic, tangentMode=RCTM_User, time=0.35, value=0}
         {interpMode=RCIM_Cubic, tangentMode=RCTM_User, time=1, value=1}
        ]}}})
    Module.Index0Param0Write=true
    Module.Index0Param1Write=true
    Module.Index0Param2Write=true
    Module.Index0Param3Write=true
   )
   /SetVariables(Particles.DynamicMaterialParameterDark:Vector4f=&MakeVector4(
     X=-0.25
     Y=&Lerp_Float(
      A=-0.1
      B=0.1
      Alpha=&FloatFromCurve(FloatCurve=DataInterface() {propertyValues {Curve {keys [<at 0 0>,<at 1 1>]}}})
     )
     Z=0.8
     W=0
    )) {isSetParameters=true}
  ]
  rendererClasses [&NiagaraMeshRendererProperties]
  renderers [
   {rendererClass=&NiagaraMeshRendererProperties, rendererIndex=0}
   {rendererClass=&NiagaraMeshRendererProperties, rendererIndex=1}
  ]
 }}
summary {
 systemName=NS_WinterDecay_Mask_TeleportDirection
 emitters [{emitterName=DirectionTendrils, rendererClasses [&NiagaraMeshRendererProperties]}]
 userVariables [
  User.PortalPosition {type=NiagaraPosition, description="", defaultValue {}}
  User.Variant {type=NiagaraInt32}
 ]
}
