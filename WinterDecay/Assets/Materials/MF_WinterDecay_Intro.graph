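# WinterDecay intro material function that outputs Emissive and OpacityMask for a dissolve in and out, timed by two custom primitive data timestamps.
# Fade progress drives a radial gradient and a local-Y "vertical" gradient from pre-skinned position, combined by Max into the mask and a glowing edge band.
# Fire noise scrolling along Z, sampled on XZ and YZ planes, breaks up both outputs.
# The cyan emissive edge is brightest at low fade progress.
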MaterialFunction /WinterDecay/Assets/Materials/MF_WinterDecay_Intro
graph {nodes [
  Constant()
  Frac(None=Divide(A=Time() {period=1}) {constB=1})
  n1 = VertexInterpolator(VS=LocalPosition() {localOrigin=InstancePreSkinning}.XYZ)
  n2 = OneMinus(None=Distance(A=@n1.PS, B=Constant()))
  n3 = Constant() {r=20}
  n4 = ScalarParameter() {defaultValue=-2.3, parameterName=FadeLerpRadial1}
  n5 = ScalarParameter() {defaultValue=1.3, parameterName=FadeLerpRadial2}
  n6 = Subtract(A=Time() {period=1}, B=ScalarParameter() {bUseCustomPrimitiveData=true, parameterName=TimeStampIn})
  n7 = ScalarParameter() {defaultValue=0.7, parameterName=FadeSpeedIn}
  n8 = Divide(
   A=Subtract(A=Time() {period=1}, B=ScalarParameter() {bUseCustomPrimitiveData=true, parameterName=TimeStamp1Out, primitiveDataIndex=1})
   B=ScalarParameter() {defaultValue=0.7, parameterName=FadeSpeedOut}
  )
  n9 = Frac(None=Multiply(A=Saturate(None=@n8)) {constB=1.01})
  n10 = Saturate(None=Min(A=Divide(A=@n6, B=@n7), B=OneMinus(None=@n9)))
  n11 = Min(A=@n10) {constB=0.5}
  n12 = ComponentMask(None=@n1.PS) {g=true}
  n13 = Constant() {r=20}
  n14 = ScalarParameter() {defaultValue=-2, parameterName=FadeVertical1}
  n15 = ScalarParameter() {defaultValue=1.4, parameterName=FadeVertical2}
  n16 = Max(
   A=Add(A=Divide(A=@n2, B=@n3), B=LinearInterpolate(A=@n4, B=@n5, Alpha=@n11) {constA=0.352, constB=0.69}) {constB=0.66}
   B=Add(A=Divide(A=@n12, B=@n13), B=LinearInterpolate(A=@n14, B=@n15, Alpha=@n10) {constA=0.3, constB=0.66}) {constB=0.66}
  )
  n17 = OneMinus(None=Saturate(None=Abs(None=@n16)))
  n18 = Multiply(A=@n17, B=@n17)
  n19 = LocalPosition() {localOrigin=InstancePreSkinning}
  n20 = Multiply(A=Time(), B=Constant3Vector() {constant=<rgba 0 0 -10 0>})
  n21 = Divide(A=VertexInterpolator(VS=Add(A=@n19.XYZ, B=@n20)).PS, B=Constant() {r=70})
  n22 = Max(
   A=TextureSample(UVs=ComponentMask(None=@n21) {b=true, r=true}) {texture=/Game/Global/Textures/Default/Noise/T_FireNoise_01}.R
   B=TextureSample(UVs=ComponentMask(None=@n21) {b=true, g=true}) {texture=&T_FireNoise_01}.R
  )
  n23 = Add(A=@n18, B=Power(Base=@n22) {constExponent=4})
  n24 = VectorParameter() {parameterName=Emission Color, defaultValue=<rgba 1.9 13 20 1>}
  n25 = OneMinus(None=@n10)
  n26 = Add(A=Multiply(A=Multiply(A=@n25, B=@n25)) {constB=10})
  FunctionOutput(None=Multiply(A=Multiply(A=@n18, B=@n23), B=Multiply(A=@n24.RGB, B=@n26))) {outputName=Emissive}
  n27 = Saturate(None=@n16)
  FunctionOutput(None=Multiply(A=@n27, B=Add(A=@n27, B=@n22))) {outputName=OpacityMask, sortPriority=1}
  Multiply(A=@n22, B=@n22)
  OneMinus(None=ComponentMask(None=@n1.PS) {b=true})
 ]}
