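# M_WinterDecay is a masked, Default Lit character material that layers a WinterDecay intro effect over the shared 2023 character parent shading.
# The parent's attributes pass through an optional static switch for vertex-color element hiding via World Position Offset, then through accessory tech-art overrides.
# MF_WinterDecay_Intro adds to the emissive and replaces the opacity mask, and the result feeds MF_CreativeMoveTool.
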Material /WinterDecay/Assets/Materials/M_WinterDecay
graph {
 properties {
  bAutomaticallySetUsageInEditor=false
  bUseMaterialAttributes=true
  bUsedWithClothing=true
  bUsedWithMeshDeformer=true
  bUsedWithMorphTargets=true
  bUsedWithNanite=true
  bUsedWithSkeletalMesh=true
  blendMode=BLEND_Masked
  pixelDepthOffsetMode=PDOM_Legacy
  refractionMethod=RM_IndexOfRefraction
  usedShadingModels=Default Lit
 }
 nodes [
  n1 = GetMaterialAttributes(None=MaterialFunctionCall() {materialFunction=&MF_CharacterParent_2023_PreFX}.Result) {attributeGetTypes [#1]}
  n2 = MaterialFunctionCall!(WPO) {materialFunction=/Game/Packages/Fortress_SharedMaterials/Base_Material_Functions/Fortnite_HideElementsWithVertexColor}
  n3 = MaterialFunctionCall(In=SetMaterialAttributes!(
    MaterialAttributes=@n1.MaterialAttributes
    World Position Offset=StaticSwitchParameter(True=@n2.Result, False=@n1.World Position Offset) {group=Opacity Mask, parameterName=Use Vertex Colors for Mask}
   ) {attributeSetTypes [#1]}) {materialFunction=&MF_TechArtOverrides_accessories}
  n4 = GetMaterialAttributes(None=@n3.Result) {attributeGetTypes [#2]}
  n5 = MaterialFunctionCall() {materialFunction=&/MF_WinterDecay_Intro}
  MaterialFunctionCall(MA In=SetMaterialAttributes!(MaterialAttributes=@n4.MaterialAttributes, Emissive Color=Add(A=@n4.Emissive Color, B=@n5.Emissive), Opacity Mask=@n5.OpacityMask) {attributeSetTypes [#2, #3]}) {materialFunction=&MF_CreativeMoveTool}
 ]
}
