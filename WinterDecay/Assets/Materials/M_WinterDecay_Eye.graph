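# Masked eye material for WinterDecay skeletal meshes, with a tintable iris, procedural pupil and glints.
# View-dependent bump-offset parallax sets iris depth, while polar math builds polygonal pupil and ring masks, with shape blends driven by time of day.
# Pupil normals blend with EyeNormal, and shading adds sun-reflection highlights and Fresnel grazing tint.
# Panners generate a darting Jitter UV, and MF_WinterDecay_Intro supplies the opacity mask plus emissive added to the glints.
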Material /WinterDecay/Assets/Materials/M_WinterDecay_Eye
graph {
 properties {
  bAutomaticallySetUsageInEditor=false
  bUseMaterialAttributes=true
  bUsedWithMeshDeformer=true
  bUsedWithSkeletalMesh=true
  blendMode=BLEND_Masked
  usedShadingModels=Default Lit
 }
 nodes [
  Constant()
  Constant()
  n1 = TextureSampleParameter2D() {parameterName=EyeNormal, samplerType=SAMPLERTYPE_Normal, texture=&FlatNormalBlue}
  n2 = DotProduct(A=Multiply(A=Constant3Vector() {constant=<rgba 0.05 0.05 1 0>}, B=@n1.RGB), B=CameraVectorWS())
  n3 = OneMinus(None=Clamp(None=Multiply(A=@n2) {constB=5}) {maxDefault=1.1, minDefault=0.8})
  n4 = Add(A=NamedRerouteUsage(), B=Multiply(A=@n3, B=@n3))
  n5 = Divide(A=Constant() {r=4}) {constA=4, constB=10}
  n6 = Divide(A=@n5)
  n7 = TextureCoordinate()
  n8 = Add(A=Multiply(A=@n7) {constB=0.1}) {constB=0.52}
  n9 = DotProduct(A=CameraVectorWS(), B=VertexNormalWS())
  n10 = LinearInterpolate(
   A=ScalarParameter() {defaultValue=-1, parameterName=DepthFront}
   B=ScalarParameter() {defaultValue=-15, parameterName=DepthSide}
   Alpha=Power(Base=OneMinus(None=@n9))
  ) {constA=-10, constB=-15}
  n11 = SphereMask(A=Add(A=@n7) {constB=-0.5}, B=Constant2Vector()) {attenuationRadius=0.5, hardnessPercent=0}
  n12 = Add(A=NamedRerouteUsage(), B=BumpOffset(Coordinate=Add(A=@n8) {constB=-0.5}, Height=Multiply(A=@n10, B=@n11)) {heightRatio=0.01})
  n13 = Add(A=Multiply(A=@n12, B=Constant2Vector() {g=1.5, r=2}), B=Constant2Vector() {g=-0.3, r=-0.5})
  n14 = Constant2Vector() {g=-0.01}
  n15 = Constant4Vector() {constant=<rgba 0.5 0.2 0.75 1>}
  n16 = NamedRerouteDeclaration(None=MaterialFunctionCall(Value_Morning=@n15.R, Value_Day=@n15.R, Value_Evening=@n15.R, Value_Night=@n15.R) {materialFunction=&Scalar_Time_of_Day}.Result) {name=TOD_ShapeLerp, nodeColor=<rgba 0 1 0.7422 1>}
  n17 = MaterialFunctionCall(UVs=LinearInterpolate(A=@n13, B=Add(A=@n12, B=@n14), Alpha=@n16), "Rotation Angle (0-1)"=Constant()) {materialFunction=&CustomRotator}
  n18 = Subtract(A=@n17.Rotated Values, B=Constant2Vector() {g=0.5, r=0.5})
  n19 = Length(None=@n18) {is=class MaterialExpressionLength}
  n20 = Constant() {r=5}
  n21 = Subtract(A=@n20, B=Constant() {r=15})
  n22 = Constant2Vector() {g=2, r=2}
  n23 = Multiply(A=Multiply(A=@n18, B=@n22), B=@n22) {desc=*UVScale}
  n24 = Arctangent2!(Y=ComponentMask(None=@n23) {r=true}, X=ComponentMask(None=@n23) {g=true})
  n25 = MaterialFunctionCall(Multiplier=Constant() {r=2}) {materialFunction=&Pi}
  n26 = Constant() {r=3}
  n27 = Multiply(A=Add(A=Divide(A=@n24, B=@n25.Result), B=Constant() {r=0.5}), B=@n26)
  n28 = Multiply(A=Subtract(A=Frac(None=@n27) {desc=t.x}, B=Constant() {desc=randMix, r=0.5}), B=@n25.Result)
  n29 = Divide(A=Divide(B=Length(None=@n23)) {desc=t.y}, B=Cosine(None=Divide(A=@n28, B=@n26)) {desc=Period:2pi, period=6.283})
  n30 = LinearInterpolate(Alpha=NamedRerouteUsage()) {constA=0.02, constB=0.01}
  n31 = Multiply(
   A=Divide(A=@n5, B=@n19) {constA=0.06}
   B=Multiply(A=SmoothStep(Min=@n21, Max=@n20, Value=@n29) {constMax=5, constMin=4.5}, B=Divide(A=@n30, B=@n19) {constA=0.01})
  )
  n32 = SmoothStep(Max=@n6, Value=@n31) {constMax=0.501, constMin=0.2}
  n33 = TextureSample(UVs=@n17.Rotated Values) {samplerType=SAMPLERTYPE_Normal, texture=/Game/Global/Textures/Default/Eyes/Sphere_EYE_NORMALS}
  n34 = LinearInterpolate(A=Constant() {r=-0.5}, B=AppendVector(A=Multiply(A=@n32, B=@n33.R), B=Multiply(A=@n32, B=@n33.G)), Alpha=@n32)
  n35 = Constant() {r=1}
  n36 = Constant2Vector() {g=1, r=4}
  n37 = Multiply(A=@n12, B=Constant2Vector() {g=1.25, r=1.25})
  n38 = Constant2Vector() {g=-0.125, r=-0.125}
  n39 = LinearInterpolate(
   A=Add(A=Multiply(A=@n12, B=@n36), B=Constant2Vector() {g=-0.05, r=-1.5})
   B=Add(A=Add(A=@n37, B=@n38), B=Constant2Vector() {g=-0.025})
   Alpha=@n16
  )
  n40 = Constant() {r=-0.08}
  n41 = NamedRerouteUsage()
  n42 = Subtract(
   A=MaterialFunctionCall(UVs=@n39, "Rotation Angle (0-1)"=LinearInterpolate(B=@n40, Alpha=@n41) {constB=3}) {materialFunction=&CustomRotator}.Rotated Values
   B=Constant2Vector() {g=0.5, r=0.5}
  )
  n43 = Constant() {r=5}
  n44 = ScalarParameter() {defaultValue=10, parameterName=EdgeSoftness}
  n45 = Constant2Vector() {g=2, r=2}
  n46 = Multiply(A=Multiply(A=@n42, B=@n45), B=@n45) {desc=*UVScale}
  n47 = Divide(B=Length(None=@n46)) {desc=t.y}
  n48 = ComponentMask(None=@n46) {r=true}
  n49 = ComponentMask(None=@n46) {g=true}
  n50 = MaterialFunctionCall(Multiplier=Constant() {r=2}) {materialFunction=&Pi}
  n51 = Add(A=Divide(A=Arctangent2(Y=@n48, X=@n49), B=@n50.Result), B=Constant() {r=0.5})
  n52 = LinearInterpolate(Alpha=NamedRerouteUsage()) {constA=4, constB=3}
  n53 = Subtract(A=Frac(None=Multiply(A=@n51, B=@n52)) {desc=t.x}, B=Constant() {desc=randMix, r=0.5})
  n54 = Cosine(None=Divide(A=Multiply(A=@n53, B=@n50.Result), B=@n52)) {desc=Period:2pi, period=6.283}
  n55 = Multiply(
   A=Divide(A=Divide(A=@n35) {constB=10}, B=Length(None=@n42)) {constA=0.06}
   B=SmoothStep(Min=Subtract(A=@n43, B=@n44), Max=@n43, Value=Divide(A=@n47, B=@n54)) {constMax=5, constMin=4.5}
  )
  n56 = Step(Y=@n35, X=@n55) {constX=0.1, constY=0.6}
  n57 = TextureSample(UVs=@n39) {samplerType=SAMPLERTYPE_Normal, texture=&Sphere_EYE_NORMALS}
  n58 = LinearInterpolate(A=Constant() {r=-0.5}, B=AppendVector(A=Multiply(A=@n56, B=@n57.R), B=Multiply(A=@n56, B=@n57.G)), Alpha=@n56)
  n59 = Saturate(None=Step(Y=@n35, X=@n55) {constX=0.1, constY=0.6})
  n60 = MaterialFunctionCall(
   Normal=LinearInterpolate(
    A=MaterialFunctionCall!(XY vector=@n34, Z sign, Default Normalization vector) {materialFunction=/Engine/Functions/Engine_MaterialFunctions02/Math/DeriveNormalZ_Function}.Result
    B=MaterialFunctionCall(XY vector=@n58) {materialFunction=&DeriveNormalZ_Function}.Result
    Alpha=NamedRerouteDeclaration(None=@n59) {name=Pupil, nodeColor=<rgba 0.9688 1 0 1>}
   )
   Flatness=Constant() {r=0.85}
  ) {materialFunction=&FlattenNormal}
  n61 = Multiply(
   A=Normalize(VectorInput=MaterialFunctionCall(BaseNormal=@n1.RGB, AdditionalNormal=@n60.Result) {materialFunction=&BlendAngleCorrectedNormals}.Result)
   B=Constant3Vector() {constant=<rgba 0.66 0.66 1 0>}
  )
  n62 = ComponentMask(None=Transform(None=@n61)) {b=true}
  Multiply(A=@n4, B=Clamp(None=ConstantBiasScale(None=@n62)))
  n63 = Constant() {r=0.5}
  n64 = ScalarParameter() {parameterName=EyeDartingTransition}
  n65 = Add(A=@n63, B=@n64)
  n66 = Subtract(A=@n63, B=@n64)
  n67 = TextureCoordinate() {uTiling=0, vTiling=0}
  n68 = Multiply(A=ScalarParameter() {defaultValue=1, parameterName=JitterAmount}, B=Time())
  n69 = Panner(Coordinate=@n67, Time=@n68) {speedX=-0.1, speedY=0.2}
  n70 = Subtract(A=Multiply(A=Frac(None=@n69)) {constB=2})
  n71 = Saturate(None=SmoothStep(Min=@n65, Max=@n66, Value=Abs(None=@n70)) {constMax=0.5, constMin=0.55})
  n72 = Add(A=Multiply(A=ComponentMask(None=@n71) {r=true}) {constB=0.2}, B=Multiply(A=ComponentMask(None=@n71) {g=true}) {constB=0.4})
  n73 = Panner(Coordinate=@n67, Time=@n68) {speedX=0.12, speedY=-0.05}
  n74 = Subtract(A=Multiply(A=Frac(None=@n73)) {constB=2})
  n75 = Saturate(None=SmoothStep(Min=@n65, Max=@n66, Value=Abs(None=@n74)) {constMax=0.5, constMin=0.55})
  n76 = Add(A=Multiply(A=ComponentMask(None=@n75) {r=true}) {constB=0.2}, B=Multiply(A=ComponentMask(None=@n75) {g=true}) {constB=0.4})
  n77 = Clamp(None=Multiply(A=AppendVector(A=@n72, B=@n76)) {constB=0.3})
  NamedRerouteDeclaration(None=Add(A=ConstantBiasScale(None=@n77) {bias=-0.5, scale=0.15}, B=TextureCoordinate())) {name=Jitter UV, nodeColor=<rgba 1 0 0.7734 1>}
  n78 = MaterialFunctionCall(UVs=@n12, Texture Scale=ScalarParameter() {defaultValue=10, group=EyeColor, parameterName=IrisRadius}) {materialFunction=&ScaleUVsByCenter}
  n79 = Transform(None=TextureSample(UVs=@n78.UVs) {samplerType=SAMPLERTYPE_Normal, texture=&Sphere_EYE_NORMALS}.RGB)
  n80 = Saturate(None=DotProduct(A=AtmosphericLightVector(), B=ReflectionVectorWS(CustomWorldNormal=@n79)))
  n81 = ScalarParameter() {defaultValue=10, group=EyeShading, parameterName=Inner HighlightRoughness}
  n82 = Multiply(
   A=@n4
   B=Multiply(A=ScalarParameter() {defaultValue=2.5, group=EyeShading, parameterName=Inner Highlight Specular}, B=Power(Base=@n80, Exp=@n81))
  )
  n83 = VectorParameter() {desc=Need for energy conservative lighting, group=EyeShading, parameterName=GrazingTint_B, defaultValue=<rgba 0.0469 0.0183 0.0221 1>}
  n84 = ScalarParameter() {defaultValue=0.7, group=EyeShading, parameterName=GrazingShadeExponent}
  NamedRerouteDeclaration(None=LinearInterpolate(A=Add(A=@n4, B=@n82), B=@n83.RGB, Alpha=Fresnel(ExponentIn=@n84) {baseReflectFraction=0}) {constB=0.25}) {name=BaseColor, nodeColor=<rgba 0 1 0.9062 1>}
  NamedRerouteDeclaration(None=@n61) {name=Normal, nodeColor=<rgba 0.5 0.5 1 1>}
  n85 = ComponentMask(None=Transform(None=ReflectionVectorWS())) {g=true, r=true}
  n86 = Constant2Vector() {g=0.3, r=0.3}
  NamedRerouteDeclaration(None=TextureSample(UVs=Multiply(A=@n85, B=@n86)) {samplerType=SAMPLERTYPE_Grayscale, texture=/Game/Global/Textures/Default/Noise/LowResBlurredNoiseNormalized}.RGB) {name=Swirls, nodeColor=<rgba 0.875 1 0 1>}
  n87 = VectorParameter() {group=0_Base, parameterName=CustomColor, sortPriority=0, defaultValue=<rgba 1 0.5701 0.0885 1>}
  n88 = Multiply(A=@n12, B=Constant2Vector() {g=1, r=0.75})
  n89 = Constant2Vector() {g=-0.01, r=0.125}
  n90 = TextureSample(UVs=LinearInterpolate(A=@n13, B=Add(A=@n88, B=@n89), Alpha=@n16), Tex=NamedRerouteUsage()) {samplerSource=SSM_Clamp_WorldGroupSettings}
  n91 = Multiply(A=@n87.RGB, B=@n90.RGB) {constB=2}
  n92 = Multiply(A=@n91, B=NamedRerouteUsage())
  n93 = LinearInterpolate(
   A=LinearInterpolate(A=Add(A=@n91, B=@n92), B=@n87.RGB, Alpha=Multiply(A=@n90.G) {constB=2})
   B=Constant3Vector() {constant=<rgba 0.02 0.02 0.02 1>}
   Alpha=NamedRerouteUsage()
  ) {constA=1}
  n94 = Constant3Vector() {constant=<rgba 0.2448 0.1426 0.2119 1>}
  n95 = NamedRerouteUsage()
  n96 = Add(A=@n78.UVs) {constB=-0.5}
  n97 = Constant() {r=0.5}
  n98 = Constant() {r=10}
  n99 = Add(
   A=Step(X=Length(None=@n96)) {constX=0.2, constY=0.2}
   B=OneMinus(None=MaterialFunctionCall(Radius=@n97, Density=@n98) {materialFunction=&RadialGradientExponential}.RadialGradientExponential)
  )
  NamedRerouteDeclaration(None=LinearInterpolate(A=LinearInterpolate(A=@n93, B=@n94, Alpha=@n95) {constA=1}, Alpha=Saturate(None=@n99)) {constB=0}) {name=Eye, desc=Might go straightaway to BC, nodeColor=<rgba 0.8047 0 1 1>}
  NamedRerouteDeclaration(None=TextureObjectParameter() {parameterName=Eye, samplerType=SAMPLERTYPE_Masks, texture=&T_BlackMask}) {name=Eye tex, nodeColor=<rgba 0.2188 1 0 1>}
  n100 = MaterialFunctionCall() {materialFunction=&/MF_WinterDecay_Intro}
  n101 = Constant2Vector() {g=0.5, r=0.5}
  n102 = Constant2Vector() {g=0.03, r=0.02}
  n103 = Constant2Vector() {g=0.03, r=0.04}
  n104 = NamedRerouteUsage()
  n105 = Length(None=Add(A=Subtract(A=@n12, B=@n101), B=LinearInterpolate(A=@n102, B=@n103, Alpha=@n104)) {constB=0.05})
  n106 = SmoothStep(Value=@n105) {constMax=0.005, constMin=0.05}
  n107 = Multiply(A=Multiply(A=@n106, B=@n106) {constB=2}, B=Constant3Vector() {constant=<rgba 0.4437 0.6744 0.9062 1>})
  n108 = Multiply(A=Power(Base=@n9) {constExponent=3}, B=Constant3Vector() {constant=<rgba 1 0.8726 0 1>})
  n109 = NamedRerouteUsage()
  n110 = Constant4Vector() {constant=<rgba 0.1 0 0.5 1>}
  n111 = Multiply(
   A=LinearInterpolate(A=Multiply(A=@n108, B=@n109), B=@n109, Alpha=NamedRerouteUsage())
   B=MaterialFunctionCall(Value_Morning=@n110.R, Value_Day=@n110.R, Value_Evening=@n110.R, Value_Night=@n110.R) {materialFunction=&Scalar_Time_of_Day}.Result
  )
  NamedRerouteDeclaration(None=Add(A=@n100.Emissive, B=Add(A=@n107, B=@n111)) {constB=0}) {name=OUT_Emissive, nodeColor=<rgba 0.0781 1 0 1>}
  NamedRerouteDeclaration(None=@n100.OpacityMask) {name=Out_OpacityMask, nodeColor=<rgba 0 0.8828 1 1>}
  n112 = Step(Y=Subtract(A=@n6) {constB=0.016}, X=@n31) {constX=0.1, constY=0.2}
  n113 = Step(Y=@n6, X=@n31) {constX=0.1, constY=0.2}
  NamedRerouteDeclaration(None=Saturate(None=Subtract(A=@n112, B=@n113))) {name=Pupil_Ring, nodeColor=<rgba 0.9688 1 0 1>}
  NamedRerouteUsage()
  NamedRerouteUsage()
  ScalarParameter() {defaultValue=10, parameterName=EdgeSoftness}
  SetMaterialAttributes!(
   MaterialAttributes
   Base Color=NamedRerouteUsage()
   Metallic=ScalarParameter() {parameterName=Metallic}
   Roughness=ScalarParameter() {defaultValue=0.1, parameterName=Roughness}
   Normal=NamedRerouteUsage()
   Emissive Color=NamedRerouteUsage()
   Opacity Mask=NamedRerouteUsage()
  ) {attributeSetTypes [
    #1
    #2
    #3
    #4
    #5
    #6
   ]}
  SmoothStep(Max=@n35, Value=@n55) {constMax=0.501, constMin=0.2}
  Step() {constX=0.1, constY=0.6}
  Step(Y=@n105) {constX=0.02}
 ]
}
