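# Animation Blueprint for the WinterDecay gadget, a custom character part that gently floats and sways with simulated physics.
# The slotted input pose is rotated and shifted by customization offsets, with bSpineSocket selecting bone-space or parent-space translation, then nudged by a fixed offset.
# Anim Dynamics springs drive root, offset and main bones, with a time-based sine drift added to main.
# The EventGraph computes that drift and binds the customization-finished delegate until PreDestruction.
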AnimBlueprint /WinterDecay/Assets/Meshes/WinterDecay_Gadget_AnimBP
target_skeleton=&/WinterDecay_Gadget_Skeleton
graphs [
 AnimGraph {nodes [
   n1 = "Transform (Modify) Bone - Bone: root"(ComponentPose=Local To Component(LocalPose=Use cached pose 'StartPose'()), Translation="0,0,0", Rotation="-90,0,0") {
    rotationMode=BMM_Additive
    rotationSpace=BCS_BoneSpace
    alphaBoolBlend {blendOption=Linear}
    boneToModify {boneName=root}
    rotation {pitch=-90}
   }
   n2 = "Transform (Modify) Bone - Bone: offset"(
    ComponentPose=@n1
    Alpha=1
    Translation=Get LocalOffsetT() -> (LocalOffsetT:Vector)
    Rotation=Get LocalOffsetR() -> (LocalOffsetR:Rotator)
    Scale="1,1,1"
   ) {
    rotationMode=BMM_Additive
    rotationSpace=BCS_ParentBoneSpace
    scaleMode=BMM_Additive
    scaleSpace=BCS_ParentBoneSpace
    translationMode=BMM_Additive
    translationSpace=BCS_BoneSpace
    alphaBoolBlend {blendOption=Linear}
    boneToModify {boneName=offset}
   }
   n3 = Component To Local(ComponentPose=@n2)
   n4 = "Transform (Modify) Bone - Bone: root"(ComponentPose=Local To Component(LocalPose=Use cached pose 'StartPose'()), Translation="0,0,0", Rotation="-90,0,0") {
    rotationMode=BMM_Additive
    rotationSpace=BCS_BoneSpace
    alphaBoolBlend {blendOption=Linear}
    boneToModify {boneName=root}
    rotation {pitch=-90}
   }
   n5 = "Transform (Modify) Bone - Bone: offset"(ComponentPose=@n4, Alpha=1, Translation=Get LocalOffsetT(), Rotation=Get LocalOffsetR(), Scale="1,1,1") {
    rotationMode=BMM_Additive
    rotationSpace=BCS_ParentBoneSpace
    scaleMode=BMM_Additive
    scaleSpace=BCS_ParentBoneSpace
    translationMode=BMM_Additive
    translationSpace=BCS_ParentBoneSpace
    alphaBoolBlend {blendOption=Linear}
    boneToModify {boneName=offset}
   }
   n6 = Component To Local(ComponentPose=@n5)
   n7 = Get bSpineSocket() -> (bSpineSocket:bool)
   n8 = "Transform (Modify) Bone - Bone: offset"(
    ComponentPose=Local To Component(LocalPose=Blend Poses by bool(BlendPose_0=@n6, BlendPose_1=@n3, bActiveValue=@n7) {blendType=Linear})
    Alpha=1
    Translation="0,9,5"
    Rotation="0,0,0"
    Scale="1,1,1"
   ) {
    rotationMode=BMM_Additive
    rotationSpace=BCS_ParentBoneSpace
    scaleMode=BMM_Additive
    scaleSpace=BCS_ParentBoneSpace
    translationMode=BMM_Additive
    translationSpace=BCS_ParentBoneSpace
    alphaBoolBlend {blendOption=Linear}
    boneToModify {boneName=offset}
    translation {y=9, z=5}
   }
   n9 = Anim Dynamics - Bone: root(ComponentPose=@n8, Alpha=1, ExternalForce="0,0,0") {
    angularDampingOverride=0.99
    angularSpringConstant=1000
    bAngularSpring=true
    bOverrideAngularDamping=true
    bUseGravityOverride=true
    gravityScale=0.35
    linearSpringConstant=1000
    simulationSpace=Actor
    alphaBoolBlend {blendOption=Linear}
    boundBone {boneName=root}
    componentLinearAccScale=<1 1 1>
    componentLinearVelScale=<1 1 1>
    gravityOverride {x=-200}
    physicsBodyDefinitions [{
      boundBone {boneName=root}
      boxExtents=<2000 2000 2000>
      constraintSetup {angularLimitsMax=<50 5 15>, angularLimitsMin=<-50 -5 -15>, angularTarget=<0 1 0>}
      localJointOffset=<0 15 0>
     }]
   }
   n10 = Anim Dynamics - Bone: offset(ComponentPose=@n9, Alpha=0.75, ExternalForce="0,0,0") {
    bAngularSpring=true
    bGravityOverrideInSimSpace=true
    bLinearSpring=true
    bOverrideLinearDamping=true
    bUseGravityOverride=true
    bUseSphericalLimits=true
    linearDampingOverride=0.9998
    linearSpringConstant=45
    alphaBoolBlend {blendOption=Linear}
    boundBone {boneName=offset}
    componentAppliedLinearAccClamp=<150000 10000 75000>
    componentLinearAccScale {x=0.04, z=0.05}
    componentLinearVelScale {x=0.04, z=0.05}
    gravityOverride {x=-100}
    physicsBodyDefinitions [{
      boundBone {boneName=offset}
      boxExtents=<60 60 60>
      constraintSetup {
       linearXLimitType=Free
       linearYLimitType=Free
       linearZLimitType=Free
       angularLimitsMax=<0 0 2>
       angularLimitsMin=<0 0 -2>
       angularTarget=<0 0 0>
       linearAxesMax=<2 2 2>
       linearAxesMin=<-2 0 -2>
      }
      localJointOffset=<0 1 0>
     }]
   }
   n11 = Anim Dynamics - Bone: main(
    ComponentPose="Transform (Modify) Bone - Bone: main"!(ComponentPose=@n10, Alpha=1, Translation=Get Float() -> (Float:Vector)) -> (Pose) {translationMode=BMM_Additive, alphaBoolBlend {blendOption=Linear}, boneToModify {boneName=main}}
    Alpha=1
    ExternalForce="0,0,0"
   ) {
    angularDampingOverride=1
    angularSpringConstant=500
    bAngularSpring=true
    bOverrideAngularDamping=true
    bUseGravityOverride=true
    gravityScale=0.35
    linearSpringConstant=1000
    simulationSpace=Actor
    alphaBoolBlend {blendOption=Linear}
    boundBone {boneName=main}
    componentLinearAccScale=<1 1 1>
    componentLinearVelScale=<1 1 1>
    gravityOverride {x=-200}
    physicsBodyDefinitions [{
      boundBone {boneName=main}
      boxExtents=<500 500 500>
      constraintSetup {angularLimitsMax=<20 0 20>, angularLimitsMin=<-20 0 -20>, angularTarget=<0 1 0>}
      localJointOffset=<0 7 0>
     }]
   }
   Output Pose(Result=Component To Local(ComponentPose=@n11))
   Save cached pose 'StartPose'(Pose=Slot 'DefaultSlot'(Source=Input Pose())) {cacheName=StartPose}
   Get LocalOffsetS() -> (LocalOffsetS:Vector)
   Get LocalOffsetS()
  ]}
 EventGraph {nodes [
   n1 = CharacterCustomizationFinished_Event() {is=title K2Node_CustomEvent}
   > NativeCharacterCustomizationFinished(self:Mask Base CCPAnim Instance Object Reference)
   Event BlueprintBeginPlay() {is=title K2Node_Event}
   > Parent: BlueprintBeginPlay() {is=title K2Node_CallParentFunction}
   > Assign Character Customization Finished!(execute:Exec, self:Custom Character Part Anim Instance Object Reference, Delegate:"Delegate (by ref)"=@n1) -> (then:Exec) {is=title K2Node_AssignDelegate}
   n2 = Event BlueprintUpdateAnimation()
   > Set Time(Time:double=float + float(A=@n2.DeltaTimeX, B=Get Time() -> (Time:double))) -> (Output_Get:double)
   n3 = float / float(A=Get Time(), B=3)
   n4 = "Sin (Radians)"(A=float * float(A=float * float(A=0.9, B=3), B=@n3))
   n5 = "Sin (Radians)"(A=float * float(A=float * float(A=0.8, B=-1), B=@n3))
   > Set Float(Float:Vector=MakeVector(X=float * float(A=@n5, B=1.25), Y=0, Z=float * float(A=@n4, B=1.25))) -> (Output_Get:Vector)
   Event PreDestruction() {is=title K2Node_Event}
   > Parent: PreDestruction() {is=title K2Node_CallParentFunction}
   > Unbind Event from Character Customization Finished(self:Custom Character Part Anim Instance Object Reference, Delegate=@n1) {is=title K2Node_RemoveDelegate}
  ]}
]
variables [
 Float {type=Vector}
 Time {type=number}
]
